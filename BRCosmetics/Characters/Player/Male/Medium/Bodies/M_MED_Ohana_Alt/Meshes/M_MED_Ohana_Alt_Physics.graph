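# Physics asset for the M_MED_Ohana_Alt male medium player body, built to simulate a dangling towel.
# Three towel capsules form a chain from the pelvis, each joint allowing wide swing on one axis, narrow swing on the other and zero twist.
# Pelvis and left-thigh capsules give the towel body shapes to collide with, linked by free angular constraints.
# An oversized zero-length root capsule, offset far below, forms a broad, nearly flat surface near root height.
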PhysicsAsset /BRCosmetics/Characters/Player/Male/Medium/Bodies/M_MED_Ohana_Alt/Meshes/M_MED_Ohana_Alt_Physics
bodies {
 dyn_towel_1 [{shapeType=Capsule, length=3.5, radius=2.5, shapeName=dyn_towel_1_capsule, center=<4.8428 0 0>, rotation=<rot 90 0 0>}]
 dyn_towel_2 [{shapeType=Capsule, length=3, radius=3, shapeName=dyn_towel_2_capsule, center=<4.8428 0 0>, rotation=<rot 90 -179.1048 175.8951>}]
 dyn_towel_3 [{shapeType=Capsule, length=2.25, radius=3.75, shapeName=dyn_towel_3_capsule, center=<4.8428 0 0>, rotation=<rot 90 0 0>}]
 pelvis [{shapeType=Capsule, length=8.0522, radius=12.5032, shapeName=pelvis_capsule, center=<-5.7801 0 0>}]
 root [{shapeType=Capsule, length=0, radius=518.755, shapeName=root_capsule, center=<0 -72.9458 -515.841>, rotation=<rot 0 -90 0>}]
 thigh_l [{shapeType=Capsule, length=33.3285, radius=12.558, shapeName=thigh_l_capsule, center=<-20 0 0>, rotation=<rot 90 0 0>}]
}
constraints [
 {bone1Name=dyn_towel_2, bone2Name=dyn_towel_1, swing1LimitDegrees=60, swing2LimitDegrees=10, twistLimitDegrees=0}
 {bone1Name=dyn_towel_3, bone2Name=dyn_towel_2, swing1LimitDegrees=60, swing2LimitDegrees=10, twistLimitDegrees=0}
 {bone1Name=dyn_towel_1, bone2Name=pelvis, swing1LimitDegrees=40, swing2LimitDegrees=10, twistLimitDegrees=0}
 {
  bone1Name=thigh_l
  bone2Name=pelvis
  swing1LimitDegrees=45
  swing1Motion=Free
  swing2LimitDegrees=45
  swing2Motion=Free
  twistLimitDegrees=45
  twistMotion=Free
 }
 {
  bone1Name=pelvis
  bone2Name=root
  swing1LimitDegrees=45
  swing1Motion=Free
  swing2LimitDegrees=45
  swing2Motion=Free
  twistLimitDegrees=45
  twistMotion=Free
 }
]
