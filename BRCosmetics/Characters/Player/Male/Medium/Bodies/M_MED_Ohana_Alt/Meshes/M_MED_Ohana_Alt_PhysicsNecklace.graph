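# Physics asset for the necklace on the M_MED_Ohana_Alt male cosmetic body, covering its dynamic chain and the upper spine.
# Five thin, progressively narrower box bodies link in series from the MSRT_chain_1 anchor, with small swing limits that tighten on one axis down the chain and zero twist limits.
# The anchor joins spine_05, and that joins spine_04, through free angular constraints, with large spine capsules forming the torso volume.
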PhysicsAsset /BRCosmetics/Characters/Player/Male/Medium/Bodies/M_MED_Ohana_Alt/Meshes/M_MED_Ohana_Alt_PhysicsNecklace
bodies {
 MSRT_chain_1 [{shapeType=Capsule, length=0.505, radius=0.505, shapeName=MSRT_chain_1_capsule, rotation=<rot 0 -90 0>}]
 dyn_chain_1 [{shapeType=Box, extentX=1.5, extentY=0.5, extentZ=11.5, shapeName=dyn_chain_1_box, center=<1 0 0>}]
 dyn_chain_2 [{shapeType=Box, extentX=1.5, extentY=0.5, extentZ=9.5, shapeName=dyn_chain_2_box, center=<1 0 0>}]
 dyn_chain_3 [{shapeType=Box, extentX=1.5, extentY=0.5, extentZ=7.5, shapeName=dyn_chain_3_box, center=<1 0 0>}]
 dyn_chain_4 [{shapeType=Box, extentX=1.5, extentY=0.5, extentZ=6, shapeName=dyn_chain_4_box, center=<1 0 0>}]
 dyn_chain_5 [{shapeType=Box, extentX=1.5, extentY=0.5, extentZ=4.5, shapeName=dyn_chain_5_box, center=<1 0 0>}]
 spine_04 [{shapeType=Capsule, length=31, radius=14, shapeName=spine_04_capsule, rotation=<rot 0 -90 0>}]
 spine_05 [{shapeType=Capsule, length=17.5, radius=26.55, shapeName=spine_05_capsule, center=<-10.5 -13.56 0>}]
}
constraints [
 {bone1Name=dyn_chain_1, bone2Name=MSRT_chain_1, swing1LimitDegrees=20, swing2LimitDegrees=20, twistLimitDegrees=0}
 {bone1Name=dyn_chain_2, bone2Name=dyn_chain_1, swing1LimitDegrees=20, swing2LimitDegrees=20, twistLimitDegrees=0}
 {bone1Name=dyn_chain_3, bone2Name=dyn_chain_2, swing1LimitDegrees=20, swing2LimitDegrees=5, twistLimitDegrees=0}
 {bone1Name=dyn_chain_4, bone2Name=dyn_chain_3, swing1LimitDegrees=20, swing2LimitDegrees=5, twistLimitDegrees=0}
 {bone1Name=dyn_chain_5, bone2Name=dyn_chain_4, swing1LimitDegrees=20, swing2LimitDegrees=5, twistLimitDegrees=0}
 {
  bone1Name=MSRT_chain_1
  bone2Name=spine_05
  swing1LimitDegrees=45
  swing1Motion=Free
  swing2LimitDegrees=45
  swing2Motion=Free
  twistLimitDegrees=45
  twistMotion=Free
 }
 {
  bone1Name=spine_05
  bone2Name=spine_04
  swing1LimitDegrees=45
  swing1Motion=Free
  swing2LimitDegrees=45
  swing2Motion=Free
  twistLimitDegrees=45
  twistMotion=Free
 }
]
